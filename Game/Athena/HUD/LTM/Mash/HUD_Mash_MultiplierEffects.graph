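# HUD widget for the Mash LTM that shows layered fireball, glow and flare effects for a multiplier level.
# SetEffectsLevel tints the effects container, reveals glow above level 0, shadow, flame and first flare above 1, second flare above 2, and switches to the large flame at 3.
# Construct resets to level 0, while PreConstruct previews DEV_LevelTest at design time.
# The graph never updates the hidden multiplier text.
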WidgetBlueprint /Game/Athena/HUD/LTM/Mash/HUD_Mash_MultiplierEffects
graphs [EventGraph {nodes [
   n1 = SetEffectsLevel!() -> (OutputDelegate, then, Level:int) {is=title K2Node_CustomEvent}
   > SetContentColorAndOpacity!(
    execute
    self:Border Object Reference=Get CommonBorder_Container() -> (CommonBorder_Container:Common Border Object Reference)
    InContentColorAndOpacity:Linear Color Structure=Select!(
     Option 0:Linear Color Structure=(R=1,G=1,B=1,A=1)
     Option 1:Linear Color Structure=(R=1,G=1,B=1,A=1)
     Option 2:Linear Color Structure=(R=0,G=0.9647,B=1,A=1)
     Option 3:Linear Color Structure=(R=1,G=0.3419,B=0.0144,A=1)
     Index:int=@n1.Level
    ) -> (ReturnValue:Linear Color Structure)
   ) -> (then)
   > SetVisibility(
    self=Get Image_Glow() -> (Image_Glow:Image Object Reference)
    InVisibility=Select!(
     Option 0:ESlateVisibility Enum=Hidden
     Option 1:ESlateVisibility Enum=HitTestInvisible
     Index:bool="integer > integer"(A=@n1.Level, B=0) {is=title K2Node_CallFunction}
    ) -> (ReturnValue:ESlateVisibility Enum)
   )
   > SetVisibility(
    self=[
     Get Image_Backgroundshadow() -> (Image_Backgroundshadow:Image Object Reference)
     Get Image_spikeball1() -> (Image_spikeball1:Image Object Reference)
     Get CommonWidgetSwitcher_FlameSize() -> (CommonWidgetSwitcher_FlameSize:Common Widget Switcher Legacy Object Reference)
    ]
    InVisibility=Select(Option 0=Hidden, Option 1=HitTestInvisible, Index="integer > integer"(A=@n1.Level, B=1))
   )
   > SetVisibility(
    self=Get Image_spikeball2() -> (Image_spikeball2:Image Object Reference)
    InVisibility=Select(Option 0=Hidden, Option 1=HitTestInvisible, Index="integer > integer"(A=@n1.Level, B=2))
   )
   > SetActiveWidgetIndex(
    self=Get CommonWidgetSwitcher_FlameSize()
    Index=Select!(Option 0:int=0, Option 1:int=0, Option 2:int=0, Option 3:int=1, Index:int=@n1.Level) -> (ReturnValue:int)
   )
   n2 = Event PreConstruct()
   > Branch(Condition=@n2.IsDesignTime)
   > SetEffectsLevel!(execute, self, Level:int=Get DEV_LevelTest() -> (DEV_LevelTest:int=0)) -> (then) {is=title K2Node_CallFunction}
   Event Construct()
   > SetEffectsLevel(Level=0)
  ]}]
variables [DEV_LevelTest {type=integer}]
widgets {
 info {inheritedWidgetCount=0, parentClass=&UserWidget, rootWidgetClass=&Overlay, widgetCount=11}
 widgets [
  Overlay_15 {widgetClassPath=&Overlay, widget=Overlay_15}
  CommonBorder_Container {
   widgetClassPath=&CommonBorder
   bIsVariable=true
   parent=Overlay_15
   widget=CommonBorder_Container
   properties {
    horizontalAlignment=HAlign_Center
    verticalAlignment=VAlign_Center
    contentColorAndOpacity=<rgba 0.6172 0.2307 1 1>
    padding=<ltrb -128 -128 -128 -128>
   }
   slot {class=OverlaySlot, properties {horizontalAlignment=HAlign_Center, verticalAlignment=VAlign_Center}}
  }
  Overlay_12 {
   widgetClassPath=&Overlay
   parent=CommonBorder_Container
   widget=Overlay_12
   slot {class=BorderSlot, properties {horizontalAlignment=HAlign_Center, verticalAlignment=VAlign_Center, padding=<ltrb -128 -128 -128 -128>}}
  }
  Image_Backgroundshadow {
   widgetClassPath=&Image
   bIsVariable=true
   parent=Overlay_12
   widget=Image_Backgroundshadow
   properties {
    renderOpacity=0.5
    brush {
     bIsDynamicallyLoaded=false
     drawAs=Image
     imageType=NoImage
     mirroring=NoMirror
     resourceName=None
     resourceObject=/Game/Athena/HUD/LTM/Wax/Materials/M_UI_TextGlow_Inst
     tiling=NoTile
     imageSize=<175.9768 144.8844>
     margin=<ltrb 0 0 0 0>
     outlineSettings {
      bUseBrushTransparency=false
      roundingType=HalfHeightRadius
      width=0
      color {colorUseRule=UseColor_Specified, specifiedColor=<rgba 0 0 0 0>}
      cornerRadii=<0 0 0 0>
     }
     tintColor {colorUseRule=UseColor_Specified, specifiedColor=<rgba 1 1 1 1>}
     uVRegion {bIsValid=false, max=<0 0>, min=<0 0>}
    }
    colorAndOpacity=<rgba 0.0052 0.0052 0.0056 1>
    renderTransform {angle=0, scale=<0.8 0.8>, shear=<0 0>, translation=<0 0>}
   }
   slot {class=OverlaySlot, properties {horizontalAlignment=HAlign_Center, verticalAlignment=VAlign_Center}}
  }
  CommonWidgetSwitcher_FlameSize {
   widgetClassPath=&CommonWidgetSwitcherLegacy
   bIsVariable=true
   parent=Overlay_12
   widget=CommonWidgetSwitcher_FlameSize
   properties {transitionDuration=0}
   slot {class=OverlaySlot, properties {horizontalAlignment=HAlign_Center, verticalAlignment=VAlign_Center, padding=<ltrb -15 -64 -15 -64>}}
  }
  Image_Flame {
   widgetClassPath=&Image
   bIsVariable=true
   parent=CommonWidgetSwitcher_FlameSize
   widget=Image_Flame
   properties {
    brush {
     bIsDynamicallyLoaded=false
     drawAs=Image
     imageType=NoImage
     mirroring=NoMirror
     resourceName=None
     resourceObject=/Game/Athena/HUD/LTM/Mash/Materials/M_UI_MultiplierFireball_2
     tiling=NoTile
     imageSize=<59.1469 93.9466>
     margin=<ltrb 0 0 0 0>
     outlineSettings {
      bUseBrushTransparency=false
      roundingType=HalfHeightRadius
      width=0
      color {colorUseRule=UseColor_Specified, specifiedColor=<rgba 0 0 0 0>}
      cornerRadii=<0 0 0 0>
     }
     tintColor {colorUseRule=UseColor_Specified, specifiedColor=<rgba 1 1 1 1>}
     uVRegion {bIsValid=false, max=<0 0>, min=<0 0>}
    }
    renderTransform {angle=0, scale=<1 1>, shear=<0 0>, translation=<9 -27>}
   }
   slot {class=WidgetSwitcherSlot}
  }
  Image_Flame_large {
   widgetClassPath=&Image
   bIsVariable=true
   parent=CommonWidgetSwitcher_FlameSize
   widget=Image_Flame_large
   properties {
    brush {
     bIsDynamicallyLoaded=false
     drawAs=Image
     imageType=NoImage
     mirroring=NoMirror
     resourceName=None
     resourceObject=/Game/Athena/HUD/LTM/Mash/Materials/M_UI_MultiplierFireball_Inst
     tiling=NoTile
     imageSize=<72.1914 101.0884>
     margin=<ltrb 0 0 0 0>
     outlineSettings {
      bUseBrushTransparency=false
      roundingType=HalfHeightRadius
      width=0
      color {colorUseRule=UseColor_Specified, specifiedColor=<rgba 0 0 0 0>}
      cornerRadii=<0 0 0 0>
     }
     tintColor {colorUseRule=UseColor_Specified, specifiedColor=<rgba 1 1 1 1>}
     uVRegion {bIsValid=false, max=<0 0>, min=<0 0>}
    }
    renderTransform {angle=0, scale=<1 1>, shear=<0 0>, translation=<9 -25>}
   }
   slot {class=WidgetSwitcherSlot}
  }
  Image_spikeball1 {
   widgetClassPath=&Image
   bIsVariable=true
   parent=Overlay_12
   widget=Image_spikeball1
   properties {
    renderOpacity=0.419
    brush {
     bIsDynamicallyLoaded=false
     drawAs=Image
     imageType=NoImage
     mirroring=NoMirror
     resourceName=None
     resourceObject=/Game/Athena/HUD/Materials/Victory/M_Flare_Spikeball_Rotating2
     tiling=NoTile
     imageSize=<275 274.3811>
     margin=<ltrb 0 0 0 0>
     outlineSettings {
      bUseBrushTransparency=false
      roundingType=HalfHeightRadius
      width=0
      color {colorUseRule=UseColor_Specified, specifiedColor=<rgba 0 0 0 0>}
      cornerRadii=<0 0 0 0>
     }
     tintColor {colorUseRule=UseColor_Specified, specifiedColor=<rgba 1 1 1 1>}
     uVRegion {bIsValid=false, max=<0 0>, min=<0 0>}
    }
   }
   slot {class=OverlaySlot, properties {horizontalAlignment=HAlign_Center, verticalAlignment=VAlign_Center, padding=<ltrb -64 -64 -64 -64>}}
  }
  Image_spikeball2 {
   widgetClassPath=&Image
   bIsVariable=true
   parent=Overlay_12
   widget=Image_spikeball2
   properties {
    brush {
     bIsDynamicallyLoaded=false
     drawAs=Image
     imageType=NoImage
     mirroring=NoMirror
     resourceName=None
     resourceObject=/Game/Athena/HUD/Materials/Victory/M_Flare_Spikeball_Rotating
     tiling=NoTile
     imageSize=<88.7095 55.1311>
     margin=<ltrb 0 0 0 0>
     outlineSettings {
      bUseBrushTransparency=false
      roundingType=HalfHeightRadius
      width=0
      color {colorUseRule=UseColor_Specified, specifiedColor=<rgba 0 0 0 0>}
      cornerRadii=<0 0 0 0>
     }
     tintColor {colorUseRule=UseColor_Specified, specifiedColor=<rgba 1 1 1 1>}
     uVRegion {bIsValid=false, max=<0 0>, min=<0 0>}
    }
    renderTransform {angle=0, scale=<2 2>, shear=<0 0>, translation=<0 0>}
   }
   slot {class=OverlaySlot, properties {horizontalAlignment=HAlign_Center, verticalAlignment=VAlign_Center}}
  }
  Image_Glow {
   widgetClassPath=&Image
   bIsVariable=true
   parent=Overlay_12
   widget=Image_Glow
   properties {
    brush {
     bIsDynamicallyLoaded=false
     drawAs=Image
     imageType=NoImage
     mirroring=NoMirror
     resourceName=None
     resourceObject=/Game/Athena/HUD/Materials/Victory/M_WinGlow_White_Inst
     tiling=NoTile
     imageSize=<269.5158 192.4113>
     margin=<ltrb 0 0 0 0>
     outlineSettings {
      bUseBrushTransparency=false
      roundingType=HalfHeightRadius
      width=0
      color {colorUseRule=UseColor_Specified, specifiedColor=<rgba 0 0 0 0>}
      cornerRadii=<0 0 0 0>
     }
     tintColor {colorUseRule=UseColor_Specified, specifiedColor=<rgba 1 1 1 1>}
     uVRegion {bIsValid=false, max=<0 0>, min=<0 0>}
    }
    colorAndOpacity=<rgba 0 0.087 0.2917 1>
    renderTransform {angle=0, scale=<0.5 0.5>, shear=<0 0>, translation=<4 0>}
   }
   slot {class=OverlaySlot, properties {horizontalAlignment=HAlign_Center, verticalAlignment=VAlign_Center, padding=<ltrb -256 -256 -256 -256>}}
  }
  Text_Multiplier {
   widgetClassPath=&CommonTextBlock
   bIsVariable=true
   parent=Overlay_15
   widget=Text_Multiplier
   properties {
    justification=Center
    text=x3
    visibility=Hidden
    font {
     bForceMonospaced=false
     bMaterialIsStencil=false
     fontMaterial=null
     fontObject=&BurbankSmall
     letterSpacing=0
     monospacedWidth=1
     size=24
     skewAmount=0
     typefaceFontName=Bold
     outlineSettings {
      bApplyOutlineToDropShadows=false
      bMiteredCorners=false
      bSeparateFillAlpha=false
      outlineMaterial=null
      outlineSize=1
      outlineColor=<rgba 0.7344 0.2323 0 1>
     }
    }
    renderTransform {angle=0, scale=<1 1>, shear=<-12 0>, translation=<0 0>}
    shadowColorAndOpacity=<rgba 0.2135 0.0586 0 0.8095>
    shadowOffset=<2.2903 2.6238>
   }
   slot {class=OverlaySlot, properties {horizontalAlignment=HAlign_Center, verticalAlignment=VAlign_Center}}
  }
 ]
}
