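# Sound cue for the big in-game score pop used by limited-time modes (LTMs).
# A fixed short delay feeds a single wave player for the ScoreLTM_InGamePop_Big_04 wave, with no randomization.
# It overrides concurrency to cap simultaneous instances at two and routes through the Fort_Enemy_Damaged_PSM sound class.
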SoundCue /Game/Athena/Sounds/LTMs/Score/ScoreLTM_InGamePop_Big_Cue
root=Delay_0
nodes [
 {id=Delay_0, children [WavePlayer_1], properties {delayMax=0.2, delayMin=0.2}}
 {id=WavePlayer_1, properties {soundWaveAssetPtr=&/ScoreLTM_InGamePop_Big_04}}
]
properties {
 bOverrideConcurrency=true
 maxDistance=2097152
 soundClassObject=/Game/Sounds/FortSoundClassesAndModes/Fort_Enemy_Damaged_PSM
 volumeMultiplier=1
 concurrencyOverrides {maxCount=2}
}
