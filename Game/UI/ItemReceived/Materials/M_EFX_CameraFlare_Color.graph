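# M_EFX_CameraFlare_Color is an additive UI-domain material in the ItemReceived UI folder, named for a colored camera flare effect.
# Its graph holds a single texture sample of T_EFX_CameraFlare_01, borrowed from the WeakPointEffect effects folder.
# No parameters, tint math or output connections appear in the graph.
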Material /Game/UI/ItemReceived/Materials/M_EFX_CameraFlare_Color
graph {
 properties {bUsedWithStaticMesh=false, blendMode=BLEND_Additive, materialDomain=MD_UI, pixelDepthOffsetMode=PDOM_Legacy, usedShadingModels=Default Lit}
 nodes [TextureSample() {texture=/Game/Effects/Fort_Effects/BluePrints/WeakPointEffect/T_EFX_CameraFlare_01}]
}
