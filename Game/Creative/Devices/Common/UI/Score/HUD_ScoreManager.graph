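# HUD score popup widget for Creative devices that shows accumulated score, a colored message and a kill-streak multiplier panel.
# A global event router listener feeds DoScoreEvent, which adds or restarts the score, animates the number and applies crit styling past a threshold.
# Each event restarts a KeepAlive timer animation; Outro fades the popup and its completion resets the score.
# Streak events drive multiplier text, effects and a dynamic-material progress bar.
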WidgetBlueprint /Game/Creative/Devices/Common/UI/Score/HUD_ScoreManager
animations [
 Intro {
  binding_count=1
  display_label=""
  length_seconds=0.05
  object=&self:Intro
  sequence {bindings [BorderContainer {
     id=#1
     possessed_class=Border
     tracks [{
       type=MovieSceneColorTrack
       display_name=Content Color and Opacity
       sections [{
         type=MovieSceneColorSection
         end_frame=1
         start_frame=0
         channels [
          Color.R {type=MovieSceneScriptingFloatChannel}
          Color.G {type=MovieSceneScriptingFloatChannel}
          Color.B {type=MovieSceneScriptingFloatChannel}
          Color.A {type=MovieSceneScriptingFloatChannel, keys [[0,0],[0.9997,3]]}
         ]
        }]
      }]
    }]}
 }
 ScoreBump {
  binding_count=1
  display_label=""
  length_seconds=0.2459
  object=&self:ScoreBump
  sequence {bindings [CommonTextBlock_ScoreValue {
     id=#2
     display_name=HorizontalBox_ScoreContainer
     possessed_class=CommonTextBlock
     tracks [{
       type=MovieScene2DTransformTrack
       sections [{
         type=MovieScene2DTransformSection
         end_frame=4
         start_frame=0
         channels [
          Translation.X {type=MovieSceneScriptingFloatChannel}
          Translation.Y {type=MovieSceneScriptingFloatChannel}
          Angle {type=MovieSceneScriptingFloatChannel}
          Scale.X {type=MovieSceneScriptingFloatChannel, keys [[0,1],[0.7023,1.4],[4.2153,0.8776],[4.918,1]]}
          Scale.Y {type=MovieSceneScriptingFloatChannel, keys [[0,1],[0.7023,1.4],[4.2153,0.8776],[4.918,1]]}
          Shear.X {type=MovieSceneScriptingFloatChannel}
          Shear.Y {type=MovieSceneScriptingFloatChannel}
         ]
        }]
      }]
    }]}
 }
 NewType {
  binding_count=1
  display_label=""
  length_seconds=0.25
  object=&self:NewType
  sequence {bindings [CommonTextBlock_ScoreType {
     id=#3
     possessed_class=CommonTextBlock
     tracks [
      {
       type=MovieSceneFloatTrack
       display_name=Render Opacity
       sections [{type=MovieSceneFloatSection, end_frame=5, start_frame=0, channels [None {type=MovieSceneScriptingFloatChannel, keys [[0,0],[5,1]]}]}]
      }
      {
       type=MovieScene2DTransformTrack
       sections [{
         type=MovieScene2DTransformSection
         end_frame=5
         start_frame=0
         channels [
          Translation.X {type=MovieSceneScriptingFloatChannel, keys [[0,56.3704],[5,0]]}
          Translation.Y {type=MovieSceneScriptingFloatChannel, const [0,2]}
          Angle {type=MovieSceneScriptingFloatChannel}
          Scale.X {type=MovieSceneScriptingFloatChannel}
          Scale.Y {type=MovieSceneScriptingFloatChannel}
          Shear.X {type=MovieSceneScriptingFloatChannel}
          Shear.Y {type=MovieSceneScriptingFloatChannel}
         ]
        }]
      }
     ]
    }]}
 }
 Outro {
  binding_count=1
  display_label=""
  length_seconds=0.55
  object=&self:Outro
  sequence {bindings [ScoreBorder {
     id=#1
     display_name=BorderContainer
     possessed_class=Border
     tracks [{
       type=MovieSceneColorTrack
       display_name=Content Color and Opacity
       sections [{
         type=MovieSceneColorSection
         end_frame=11
         start_frame=0
         channels [
          Color.R {type=MovieSceneScriptingFloatChannel}
          Color.G {type=MovieSceneScriptingFloatChannel}
          Color.B {type=MovieSceneScriptingFloatChannel}
          Color.A {type=MovieSceneScriptingFloatChannel, keys [[0,1],[11,0]]}
         ]
        }]
      }]
    }]}
 }
 KeepAlive {
  binding_count=1
  display_label=""
  length_seconds=2.5
  object=&self:KeepAlive
  sequence {
   bindings [BorderContainer {
     id=#4
     display_name=ScoreBorder
     possessed_class=Border
     tracks [{
       type=MovieSceneColorTrack
       display_name=Content Color and Opacity
       sections [{
         type=MovieSceneColorSection
         end_frame=50
         start_frame=-1
         channels [
          Color.R {type=MovieSceneScriptingFloatChannel, keys [[-0.0003,1],[49.6453,2]]}
          Color.G {type=MovieSceneScriptingFloatChannel, keys [[-0.0003,1],[49.6453,2]]}
          Color.B {type=MovieSceneScriptingFloatChannel, keys [[-0.0003,1],[49.6453,2]]}
          Color.A {type=MovieSceneScriptingFloatChannel, keys [[-0.0003,1],[49.6453,2]]}
         ]
        }]
      }]
    }]
   tracks [{
     type=MovieSceneEventTrack
     sections [{
       type=MovieSceneEventTriggerSection
       channels [None {type=MovieSceneScriptingEventChannel, const ["<Struct 'MovieSceneEvent' (0x467bba97080) {}>", 1]}]
      }]
    }]
  }
 }
 CritStyle {
  binding_count=3
  display_label=""
  length_seconds=1.0005
  object=&self:CritStyle
  sequence {
   bindings [
    SizeBox_1 {
     id=#5
     possessed_class=SizeBox
     tracks [
      {
       type=MovieSceneFloatTrack
       display_name=Width Override
       sections [{
         type=MovieSceneFloatSection
         end_frame=120
         start_frame=0
         channels [None {type=MovieSceneScriptingFloatChannel, keys [[0,30.6965],[27.966,369.4349],[75.946,577.3549]]}]
        }]
      }
      {
       type=MovieSceneFloatTrack
       display_name=Render Opacity
       sections [{
         type=MovieSceneFloatSection
         end_frame=120
         start_frame=0
         channels [None {type=MovieSceneScriptingFloatChannel, keys [[0,0],[11,1],[120.054,0]]}]
        }]
      }
     ]
    }
    Image_Glow {
     id=#6
     possessed_class=Image
     tracks [{
       type=MovieSceneFloatTrack
       display_name=Render Opacity
       sections [{type=MovieSceneFloatSection, end_frame=42, start_frame=0, channels [None {type=MovieSceneScriptingFloatChannel, keys [[0,0],[12,1]]}]}]
      }]
    }
    Image_Flare {
     id=#7
     possessed_class=Image
     tracks [
      {
       type=MovieSceneFloatTrack
       display_name=Render Opacity
       sections [{
         type=MovieSceneFloatSection
         end_frame=120
         start_frame=0
         channels [None {type=MovieSceneScriptingFloatChannel, keys [[0,0],[21,1],[120.054,0]]}]
        }]
      }
      {
       type=MovieScene2DTransformTrack
       sections [{
         type=MovieScene2DTransformSection
         end_frame=87
         start_frame=0
         channels [
          Translation.X {type=MovieSceneScriptingFloatChannel}
          Translation.Y {type=MovieSceneScriptingFloatChannel}
          Angle {type=MovieSceneScriptingFloatChannel}
          Scale.X {type=MovieSceneScriptingFloatChannel, keys [[0,0.2],[32.622,1.1489],[43.496,1]]}
          Scale.Y {type=MovieSceneScriptingFloatChannel, keys [[0,0.2],[32.622,1.1489],[43.496,1]]}
          Shear.X {type=MovieSceneScriptingFloatChannel}
          Shear.Y {type=MovieSceneScriptingFloatChannel}
         ]
        }]
      }
     ]
    }
   ]
   display_rate {numerator=120}
  }
 }
 NewStreakLevel {
  binding_count=3
  length_seconds=0.6754
  object=&self:NewStreakLevel
  sequence {bindings [
    Image_StreakGlow {
     id=#8
     possessed_class=Image
     tracks [
      {
       type=MovieSceneFloatTrack
       display_name=Render Opacity
       sections [{type=MovieSceneFloatSection, channels [None {type=MovieSceneScriptingFloatChannel, keys [[0,0],[1.0543,1],[7,1],[13.5067,0]]}]}]
      }
      {
       type=MovieScene2DTransformTrack
       sections [{
         type=MovieScene2DTransformSection
         channels [
          Translation.X {type=MovieSceneScriptingFloatChannel, const [4,1]}
          Translation.Y {type=MovieSceneScriptingFloatChannel, const [0,1]}
          Angle {type=MovieSceneScriptingFloatChannel, const [0,1]}
          Scale.X {type=MovieSceneScriptingFloatChannel, keys [[0,0.5],[2.862,1],[13.5067,1.2]]}
          Scale.Y {type=MovieSceneScriptingFloatChannel, keys [[0,0.5],[2.862,1],[13.5067,1.2]]}
          Shear.X {type=MovieSceneScriptingFloatChannel, const [0,1]}
          Shear.Y {type=MovieSceneScriptingFloatChannel, const [0,1]}
         ]
        }]
      }
     ]
    }
    HUD_Mash_MultiplierEffects {
     id=#9
     possessed_class=HUD_Mash_MultiplierEffects_C
     tracks [{
       type=MovieScene2DTransformTrack
       sections [{
         type=MovieScene2DTransformSection
         channels [
          Translation.X {type=MovieSceneScriptingFloatChannel}
          Translation.Y {type=MovieSceneScriptingFloatChannel}
          Angle {type=MovieSceneScriptingFloatChannel}
          Scale.X {type=MovieSceneScriptingFloatChannel, keys [[0,0],[4,1.5944],[7,1]]}
          Scale.Y {type=MovieSceneScriptingFloatChannel, keys [[0,0],[4,1.7134],[7,1]]}
          Shear.X {type=MovieSceneScriptingFloatChannel}
          Shear.Y {type=MovieSceneScriptingFloatChannel}
         ]
        }]
      }]
    }
    CommonTextBlock_MultiplierValue {
     id=#10
     possessed_class=CommonTextBlock
     tracks [{
       type=MovieScene2DTransformTrack
       sections [{
         type=MovieScene2DTransformSection
         channels [
          Translation.X {type=MovieSceneScriptingFloatChannel}
          Translation.Y {type=MovieSceneScriptingFloatChannel}
          Angle {type=MovieSceneScriptingFloatChannel}
          Scale.X {type=MovieSceneScriptingFloatChannel, keys [[0,1],[1.7557,2],[4.1393,2],[5.6697,1.1032],[7.6697,1.1078],[8.6697,1]]}
          Scale.Y {type=MovieSceneScriptingFloatChannel, keys [[0,1],[1.7557,2],[4.1393,2],[5.6697,1.1032],[7.6697,1.1078],[8.6697,1]]}
          Shear.X {type=MovieSceneScriptingFloatChannel}
          Shear.Y {type=MovieSceneScriptingFloatChannel}
         ]
        }]
      }]
    }
   ]}
 }
]
graphs [
 EventGraph {nodes [
   SetStyle!(
    execute
    self:Common Text Object Reference=[
     Get Number_Score() -> (Number_Score:Common Numeric Text Block Object Reference)
     Get CommonTextBlock_ScoreValue() -> (CommonTextBlock_ScoreValue:Common Text Object Reference)
    ]
    InStyle:Common Text Style Class Reference=/Game/Athena/HUD/Materials/Score/TextStyle-ScorePopup.TextStyle-ScorePopup_C
   ) -> (then)
   > PlayAnimationTimeRange!(
    execute
    self:User Widget Object Reference
    InAnimation:Widget Animation Object Reference=Get CritStyle() -> (CritStyle:Widget Animation Object Reference)
    StartAtTime:float=0.01
    EndAtTime:float=0
    NumLoopsToPlay:int=1
    PlayMode:EUMGSequencePlayMode Enum=Reverse
    PlaybackSpeed:float=1
    bRestoreState:bool=false
   ) -> (then, ReturnValue:Widget Animation Handle Structure)
   > Set isCrit(isCrit:bool=false) -> (Output_Get:bool)
   n1 = ScoreComplete() {is=title K2Node_CustomEvent}
   > Set CurrentScore_0(CurrentScore_0:int=0) -> (Output_Get:int=0)
   > Set CurrentType(CurrentType:Text) -> (Output_Get:Text)
   n2 = DoScoreEvent!() -> (
    OutputDelegate
    then
    Score:int
    ScoreText:Text
    ScoreColor:Linear Color Structure=(R=0.0159,G=0.5182,B=0.9323,A=1)
    TextColor:Linear Color Structure=(R=0.0192,G=0.0729,B=0.7292,A=1)
    RestartCombo:bool=~
   ) {is=title K2Node_CustomEvent}
   > SetVisibility(
    self=[
     Get HorizontalBox_ScoreContainer() -> (HorizontalBox_ScoreContainer:Horizontal Box Object Reference)
     Get CommonTextBlock_ScoreType() -> (CommonTextBlock_ScoreType:Common Text Object Reference)
    ]
    InVisibility=HitTestInvisible
   )
   > n3 = Branch(Condition=NOT Boolean(A=@n2.RestartCombo))
   > n4 = Set CurrentScore_0(CurrentScore_0=int + int(A=Get CurrentScore_0() -> (CurrentScore_0:int=0), B=@n2.Score) -> (ReturnValue=0) {is=title K2Node_CommutativeAssociativeBinaryOperator}) -> (Output_Get=0)
   n5 = Get CurrentScore_0() -> (CurrentScore_0=0)
   "SetText (Text)"(
    execute=[@n13.then, @n4.then]
    self=Get CommonTextBlock_ScoreValue()
    InText=Select(Option 0:Text, Option 1:Text="LOCTEXT(\"+\")", Index="integer > integer"(A=@n5)) -> (ReturnValue:Text)
   )
   > SetScoreColor!(execute, self, ScoreColor:Linear Color Structure=@n2.ScoreColor, TextColor:Linear Color Structure=@n2.TextColor) -> (then)
   n6 = Get CurrentScore_0() -> (CurrentScore_0=0)
   > n7 = InterpolateToValue!(
    execute
    self:Common Numeric Text Block Object Reference=Get Number_Score()
    TargetValue:float="To Float (Integer)"(InInt=@n6)
    MaximumInterpolationDuration:float=0.3
    MinimumChangeRate:float=1
    OutroOffset:float=0
   ) -> (then)
   "SetText (Text)"(execute=[@n7.then, @n37.then, @n36.else], self=Get CommonTextBlock_ScoreType(), InText=@n2.ScoreText)
   n8 = Get KeepAlive() -> (KeepAlive:Widget Animation Object Reference)
   > n9 = Branch(Condition=IsAnimationPlaying(InAnimation=@n8))
   > StopAnimation(InAnimation=@n8)
   > n10 = PlayAnimation(InAnimation=Get ScoreBump() -> (ScoreBump:Widget Animation Object Reference))
   PlayAnimation(execute=[@n11.then, @n10.then], InAnimation=@n8)
   > Branch(Condition="Not Equal Exactly (Text)"!(A:"Text (by ref)"=Get CurrentType() -> (CurrentType:Text), B:"Text (by ref)"=@n2.ScoreText) -> (ReturnValue:bool))
   > PlayAnimation(InAnimation=Get NewType() -> (NewType:Widget Animation Object Reference))
   > Set CurrentType(CurrentType=@n2.ScoreText)
   StopAnimation(execute=@n9.else, InAnimation=Get Outro() -> (Outro:Widget Animation Object Reference))
   > n11 = PlayAnimation(InAnimation=Get Intro())
   n12 = Set CurrentScore_0(execute=@n3.else, CurrentScore_0=@n2.Score) -> (Output_Get=0)
   > n13 = SetCurrentValue!(execute, self:Common Numeric Text Block Object Reference=Get Number_Score(), NewValue:float="To Float (Integer)"(InInt=@n12)) -> (then)
   Event Construct()
   > SetVisibility(InVisibility=HitTestInvisible)
   > SetVisibility(
    self=[
     Get HorizontalBox_ScoreContainer()
     Get VerticalBox_Streak() -> (VerticalBox_Streak:Vertical Box Object Reference)
     Get CommonTextBlock_ScoreType()
    ]
    InVisibility=Hidden
   )
   > BindToAnimationFinished(Animation=Get Outro(), Delegate=@n1)
   > n14 = GetDynamicMaterial(self=Get Image_MultiplierProgress() -> (Image_MultiplierProgress:Image Object Reference))
   > Set StreakProgressMID(StreakProgressMID:Material Instance Dynamic Object Reference=@n14) -> (Output_Get:Material Instance Dynamic Object Reference)
   > n15 = GetDynamicMaterial(self=Get Image_TimeRemaining() -> (Image_TimeRemaining:Image Object Reference))
   > Set StreakTimerMID(StreakTimerMID:Material Instance Dynamic Object Reference=@n15) -> (Output_Get:Material Instance Dynamic Object Reference)
   > n16 = StartListeningToEvent!(
    execute:Exec
    Target:Gameplay Event Router Component Object Reference=GetGlobalEventRouter()
    EventType:Script Struct Object Reference=/Game/Creative/Devices/Common/GameplayEvents/Event_Creative_BroadcastScoreToHUD
    Context:Object Reference
    EventBubblingRule:EEventBubblingRule Enum
   ) -> (
    then:Exec
    AsyncTaskProxy:Async Action Start Listening To Event Object Reference
    OnEventReceived:Exec
    Payload:Event Creative Broadcast Score To HUD Structure
   )
   n17 = Break Event Creative Broadcast Score to HUD!(Event_Creative_BroadcastScoreToHUD:"Event Creative Broadcast Score To HUD Structure (by ref)"=@n16.Payload) -> (
    Score_5_#11:int
    Message_15_#12:Text
    MessageColor_14_#13:Linear Color Structure
    ScoreColor_13_#14:Linear Color Structure
    RestartCombo_12_#15:bool=~
   ) {is=title K2Node_BreakStruct}
   DoScoreEvent!(
    execute=@n16.OnEventReceived
    self
    Score:int=@n17
    ScoreText:Text=@n17.Message_15_#12
    ScoreColor:Linear Color Structure=@n17.MessageColor_14_#13
    TextColor:Linear Color Structure=@n17.ScoreColor_13_#14
    RestartCombo:bool=@n17.RestartCombo_12_#15
   ) -> (then) {is=title K2Node_CallFunction}
   K2Node_Event_13 = OnStreakScoreMultiplierUpdated_1!() -> (OutputDelegate, then, StreakScoreLevel:Byte, MaxStreakScoreLevel:Byte, ScoreMultiplier:float=~) {is=title K2Node_CustomEvent}
   n18 = "To Integer (Byte)"!(InByte:Byte=@K2Node_Event_13.StreakScoreLevel) -> (ReturnValue:int=0)
   > SetEffectsLevel!(
    execute
    self:HUD Mash Multiplier Effects Object Reference=Get HUD_Mash_MultiplierEffects() -> (HUD_Mash_MultiplierEffects:HUD Mash Multiplier Effects Object Reference)
    Level:int=@n18
   ) -> (then)
   n19 = "To LinearColor (Color)"!(InColor:Color Structure="To Color (LinearColor)"!(InLinearColor:Linear Color Structure=(R=0.0003,G=0.15,B=0.956,A=1), InUseSRGB:bool=true) -> (ReturnValue:Color Structure)) -> (ReturnValue:Linear Color Structure)
   n20 = "To Color (LinearColor)"(InLinearColor=(R=1,G=0.4852,B=0,A=1), InUseSRGB=true)
   n21 = Get StreakProgressMID() -> (StreakProgressMID:Material Instance Dynamic Object Reference)
   > SetVectorParameterValue(
    self=@n21
    ParameterName=FillColor
    Value=Select!(
     Option 0:Linear Color Structure=@n19
     Option 1:Linear Color Structure=@n19
     Option 2:Linear Color Structure=@n19
     Option 3:Linear Color Structure="To LinearColor (Color)"(InColor=@n20)
     Index:int=@n18
    ) -> (ReturnValue:Linear Color Structure)
   )
   n22 = "To LinearColor (Color)"(InColor="To Color (LinearColor)"(InLinearColor=(R=0,G=0.7758,B=1,A=1), InUseSRGB=true))
   n23 = "To Color (LinearColor)"(InLinearColor=(R=1,G=1,B=0,A=1), InUseSRGB=true)
   > SetVectorParameterValue(
    self=@n21
    ParameterName=FillColor2
    Value=Select(Option 0=@n22, Option 1=@n22, Option 2=@n22, Option 3="To LinearColor (Color)"(InColor=@n23), Index=@n18)
   )
   > n24 = Branch(Condition="float > float"(A=@K2Node_Event_13.ScoreMultiplier, B=Get MultiplierValue() -> (MultiplierValue:double)) {is=title K2Node_CallFunction})
   > n25 = PlayAnimation(InAnimation=Get NewStreakLevel() -> (NewStreakLevel:Widget Animation Object Reference))
   n26 = Set MultiplierValue(execute=[@n25.then, @n24.else], MultiplierValue:double=@K2Node_Event_13.ScoreMultiplier) -> (Output_Get:double)
   > "SetText (Text)"(
    self=Get CommonTextBlock_MultiplierValue() -> (CommonTextBlock_MultiplierValue:Common Text Object Reference)
    InText=Format Text!(Format="LOCTEXT(\"x{MULT}\")", MULT:double=@n26) -> (Result)
   )
   > Branch(Condition=Get ChangeMultiplierTextColor Const() -> (ChangeMultiplierTextColor Const:bool))
   > n27 = For Each Loop with Break!(
    Exec:Exec
    Array:Array of STRUCT Mash Score Events Multiplier Colors Structures=Get Color Data Array Const() -> (Color Data Array Const:Array of STRUCT Mash Score Events Multiplier Colors Structures)
    Break:Exec=@n29.then
   ) -> (LoopBody:Exec, Array Element:STRUCT Mash Score Events Multiplier Colors Structure, Array Index:int, Completed:Exec)
   n28 = Break STRUCT Mash Score Events Multiplier Colors!(STRUCT_Mash_ScoreEvents_MultiplierColors:"STRUCT Mash Score Events Multiplier Colors Structure (by ref)"=@n27) -> (
    MaxMultiplier_2_#16:float=~
    ValueColor_5_#17:Slate Color Structure
    TypeColor_7_#18:Slate Color Structure
   ) {is=title K2Node_BreakStruct}
   > Branch(Condition="float <= float"!(A:double=Get MultiplierValue(), B:double=@n28) -> (ReturnValue:bool) {is=title K2Node_CallFunction})
   > SetColorAndOpacity(self=Get CommonTextBlock_MultiplierValue(), InColorAndOpacity=@n28.ValueColor_5_#17)
   > n29 = SetColorAndOpacity(
    self=Get CommonTextBlock_MultiplierType() -> (CommonTextBlock_MultiplierType:Common Text Object Reference)
    InColorAndOpacity=@n28.TypeColor_7_#18
   )
   K2Node_Event_8 = OnStreakScoreMultiplierKillCountUpdated_1!() -> (OutputDelegate, then, TotalStreakKillCount:int, StreakLevelProgressRatio:float=~) {is=title K2Node_CustomEvent}
   > n30 = Branch(Condition="Equal (Integer)"(A=@K2Node_Event_8.TotalStreakKillCount, B=0) {is=title K2Node_CallFunction})
   > n31 = Set CurrentScore_0(CurrentScore_0=0) -> (Output_Get=0)
   n32 = Branch(
    execute=[@n31.then, @n30.else]
    Condition="integer >= integer"!(A:int=@K2Node_Event_8.TotalStreakKillCount, B:int=2) -> (ReturnValue:bool) {is=title K2Node_CallFunction}
   )
   n33 = Get VerticalBox_Streak()
   > SetVisibility(self=@n33, InVisibility=HitTestInvisible)
   > SetScalarParameterValue(self=Get StreakProgressMID(), ParameterName=PercentFill, Value=@K2Node_Event_8.StreakLevelProgressRatio)
   > "SetText (Text)"(
    self=Get CommonTextBlock_MultiplierType()
    InText=Format Text!(Format="LOCTEXT(\"{KILL} KO STREAK\")", KILL:int=@K2Node_Event_8.TotalStreakKillCount) -> (Result)
   )
   SetVisibility(execute=@n32.else, self=@n33, InVisibility=Hidden)
   n34 = Get isCrit() -> (isCrit:bool)
   n35 = Get CritThreshold() -> (CritThreshold:int=0)
   n36 = Branch(Condition=AND Boolean(A="integer >= integer"(A=@n6, B=@n35), B=NOT Boolean(A=@n34)))
   > PlayAnimation(InAnimation=Get CritStyle())
   > SetStyle(
    self=[Get Number_Score(), Get CommonTextBlock_ScoreValue()]
    InStyle=/Game/Athena/HUD/Materials/Score/TextStyle-ScorePopup_Crit.TextStyle-ScorePopup_Crit_C
   )
   > n37 = Set isCrit(isCrit=true)
  ]}
 IncrementSoundPitch {nodes [
   IncrementSoundPitch() {is=title K2Node_FunctionEntry}
   n1 = float + float(A=Get SoundFeedbackPitchValue() -> (SoundFeedbackPitchValue:double), B=0.1) {is=title K2Node_CommutativeAssociativeBinaryOperator}
   > Set SoundFeedbackPitchValue(SoundFeedbackPitchValue:double=Select!(Option 0:double=@n1, Option 1:double=1, Index:bool="float > float"(A=@n1, B=1.3)) -> (ReturnValue:double)) -> (Output_Get:double)
  ]}
 ResetSoundPitch {nodes [
   ResetSoundPitch() {is=title K2Node_FunctionEntry}
   > Set SoundFeedbackPitchValue(SoundFeedbackPitchValue=1)
  ]}
 Sequencer Events {nodes [
   TimeOut() {is=title K2Node_CustomEvent}
   > PlayAnimation(InAnimation=Get Outro())
  ]}
 SetScoreColor {nodes [
   SetScoreColor!() -> (then, ScoreColor:Linear Color Structure=(R=0,G=0.6244,B=1,A=1), TextColor:Linear Color Structure=(R=0,G=0.0324,B=0.9688,A=1)) {is=title K2Node_FunctionEntry}
   > SetColorAndOpacity(
    self=[Get Number_Score(), Get CommonTextBlock_ScoreValue()]
    InColorAndOpacity=Make Slate Color(SpecifiedColor=Get ScoreColor() -> (ScoreColor:Linear Color Structure))
   )
   > SetColorAndOpacity(
    self=Get CommonTextBlock_ScoreType()
    InColorAndOpacity=Make Slate Color(SpecifiedColor=Get TextColor() -> (TextColor:Linear Color Structure))
   )
   > Return Node()
  ]}
]
variables [
 ChangeMultiplierTextColor Const {type=boolean}
 Color Data Array Const {
  type="array<STRUCT_Mash_ScoreEvents_MultiplierColors>"
  default [
   {
    max Multiplier=4
    type Color {colorUseRule=UseColor_Specified, specifiedColor=<rgba 1 0.9578 0.7186 1>}
    value Color {colorUseRule=UseColor_Specified, specifiedColor=<rgba 1 0.9578 0.7186 1>}
   }
   {
    max Multiplier=9
    type Color {colorUseRule=UseColor_Specified, specifiedColor=<rgba 0 0.136 0.5938 1>}
    value Color {colorUseRule=UseColor_Specified, specifiedColor=<rgba 0 0.136 0.5938 1>}
   }
   {
    max Multiplier=14
    type Color {colorUseRule=UseColor_Specified, specifiedColor=<rgba 0.4167 0.0519 0.3873 1>}
    value Color {colorUseRule=UseColor_Specified, specifiedColor=<rgba 0.4167 0.0519 0.3873 1>}
   }
   {
    max Multiplier=10000000000
    type Color {colorUseRule=UseColor_Specified, specifiedColor=<rgba 1 0.5536 0 1>}
    value Color {colorUseRule=UseColor_Specified, specifiedColor=<rgba 1 0.5536 0 1>}
   }
  ]
 }
 CritThreshold {type=integer, default=1500}
 CurrentScore_0 {type=integer}
 CurrentType {type=string, default=""}
 isCrit {type=boolean}
 LastDevice {type=Actor, default=null}
 MultiplierValue {type=number, default=1}
 SoundFeedbackPitchValue {type=number, default=1}
 SoundOnCrit {type=SoundBase, default=/Game/Athena/Sounds/LTMs/Score/ScoreLTM_InGamePop_Big_Cue}
 SoundOnNonCrit {type=SoundBase, default=/Game/Athena/Sounds/LTMs/Score/ScoreLTM_InGamePop_Little_Cue}
 StreakProgressMID {type=MaterialInstanceDynamic}
 StreakTimerMID {type=MaterialInstanceDynamic}
]
widgets {
 info {inheritedWidgetCount=0, parentClass=&UserWidget, rootWidgetClass=&Border, widgetCount=32}
 widgets [
  BorderContainer {
   widgetClassPath=&Border
   widget=BorderContainer
   properties {
    background {
     bIsDynamicallyLoaded=false
     drawAs=NoDrawType
     imageType=NoImage
     mirroring=NoMirror
     resourceName=None
     resourceObject=null
     tiling=NoTile
     imageSize=<32 32>
     margin=<ltrb 0 0 0 0>
     outlineSettings {
      bUseBrushTransparency=false
      roundingType=HalfHeightRadius
      width=0
      color {colorUseRule=UseColor_Specified, specifiedColor=<rgba 0 0 0 0>}
      cornerRadii=<0 0 0 0>
     }
     tintColor {colorUseRule=UseColor_Specified, specifiedColor=<rgba 1 1 1 1>}
     uVRegion {bIsValid=false, max=<0 0>, min=<0 0>}
    }
    padding=<ltrb 0 0 0 0>
   }
  }
  Overlay_0 {widgetClassPath=&Overlay, parent=BorderContainer, widget=Overlay_0, slot {class=BorderSlot, properties {padding=<ltrb 0 0 0 0>}}}
  VerticalBox_669 {
   widgetClassPath=&VerticalBox
   parent=Overlay_0
   widget=VerticalBox_669
   slot {class=OverlaySlot, properties {horizontalAlignment=HAlign_Fill, verticalAlignment=VAlign_Fill}}
  }
  ScoreBorder {
   widgetClassPath=&Border
   parent=VerticalBox_669
   widget=ScoreBorder
   properties {horizontalAlignment=HAlign_Left, verticalAlignment=VAlign_Top, brushColor=<rgba 1 1 1 0>, padding=<ltrb 0 0 0 0>}
   slot {class=VerticalBoxSlot}
  }
  VerticalBox_100 {
   widgetClassPath=&VerticalBox
   parent=ScoreBorder
   widget=VerticalBox_100
   slot {class=BorderSlot, properties {horizontalAlignment=HAlign_Left, verticalAlignment=VAlign_Top, padding=<ltrb 0 0 0 0>}}
  }
  VerticalBox_1 {
   widgetClassPath=&VerticalBox
   parent=VerticalBox_100
   widget=VerticalBox_1
   slot {class=VerticalBoxSlot, properties {horizontalAlignment=HAlign_Left, verticalAlignment=VAlign_Top}}
  }
  SizeBox_0 {
   widgetClassPath=&SizeBox
   parent=VerticalBox_1
   widget=SizeBox_0
   properties {bOverride_HeightOverride=true, heightOverride=82.6868}
   slot {class=VerticalBoxSlot, properties {horizontalAlignment=HAlign_Left, verticalAlignment=VAlign_Top}}
  }
  Overlay_1 {
   widgetClassPath=&Overlay
   parent=SizeBox_0
   widget=Overlay_1
   slot {class=SizeBoxSlot, properties {verticalAlignment=VAlign_Bottom}}
  }
  Image_Flare {
   widgetClassPath=&Image
   parent=Overlay_1
   widget=Image_Flare
   properties {
    renderOpacity=0
    brush {
     bIsDynamicallyLoaded=false
     drawAs=Image
     imageType=NoImage
     mirroring=NoMirror
     resourceName=None
     resourceObject=/Game/UI/ItemReceived/Materials/M_EFX_CameraFlare_Color
     tiling=NoTile
     imageSize=<252.1309 220.9598>
     margin=<ltrb 0 0 0 0>
     outlineSettings {
      bUseBrushTransparency=false
      roundingType=HalfHeightRadius
      width=0
      color {colorUseRule=UseColor_Specified, specifiedColor=<rgba 0 0 0 0>}
      cornerRadii=<0 0 0 0>
     }
     tintColor {colorUseRule=UseColor_Specified, specifiedColor=<rgba 1 1 1 1>}
     uVRegion {bIsValid=false, max=<0 0>, min=<0 0>}
    }
   }
   slot {class=OverlaySlot, properties {horizontalAlignment=HAlign_Center, verticalAlignment=VAlign_Center, padding=<ltrb -300 -300 -300 -300>}}
  }
  SizeBox_1 {
   widgetClassPath=&SizeBox
   parent=Overlay_1
   widget=SizeBox_1
   properties {bOverride_WidthOverride=true, renderOpacity=0, widthOverride=369.4349}
   slot {class=OverlaySlot, properties {horizontalAlignment=HAlign_Center, verticalAlignment=VAlign_Center, padding=<ltrb -500 0 -500 0>}}
  }
  HorizontalBox_0 {widgetClassPath=&HorizontalBox, parent=SizeBox_1, widget=HorizontalBox_0, slot {class=SizeBoxSlot}}
  ImageTriangleL {
   widgetClassPath=&Image
   parent=HorizontalBox_0
   widget=ImageTriangleL
   properties {
    brush {
     bIsDynamicallyLoaded=false
     drawAs=Image
     imageType=NoImage
     mirroring=NoMirror
     resourceName=None
     resourceObject=/Game/UI/HUD/Materials/M_FX_TriangleSpiral
     tiling=NoTile
     imageSize=<200 32>
     margin=<ltrb 0 0 0 0>
     outlineSettings {
      bUseBrushTransparency=false
      roundingType=HalfHeightRadius
      width=0
      color {colorUseRule=UseColor_Specified, specifiedColor=<rgba 0 0 0 0>}
      cornerRadii=<0 0 0 0>
     }
     tintColor {colorUseRule=UseColor_Specified, specifiedColor=<rgba 0 0.3944 1 1>}
     uVRegion {bIsValid=false, max=<0 0>, min=<0 0>}
    }
    renderTransform {angle=0, scale=<-1 1>, shear=<0 0>, translation=<0 0>}
    renderTransformPivot=<0.5 1>
   }
   slot {class=HorizontalBoxSlot, properties {verticalAlignment=VAlign_Center, size {sizeRule=Fill, value=1}}}
  }
  ImageTriangleR {
   widgetClassPath=&Image
   parent=HorizontalBox_0
   widget=ImageTriangleR
   properties {
    brush {
     bIsDynamicallyLoaded=false
     drawAs=Image
     imageType=NoImage
     mirroring=NoMirror
     resourceName=None
     resourceObject=&M_FX_TriangleSpiral
     tiling=NoTile
     imageSize=<200 32>
     margin=<ltrb 0 0 0 0>
     outlineSettings {
      bUseBrushTransparency=false
      roundingType=HalfHeightRadius
      width=0
      color {colorUseRule=UseColor_Specified, specifiedColor=<rgba 0 0 0 0>}
      cornerRadii=<0 0 0 0>
     }
     tintColor {colorUseRule=UseColor_Specified, specifiedColor=<rgba 0 0.3944 1 1>}
     uVRegion {bIsValid=false, max=<0 0>, min=<0 0>}
    }
    renderTransformPivot=<0.5 1>
   }
   slot {class=HorizontalBoxSlot, properties {size {sizeRule=Fill, value=1}}}
  }
  HorizontalBox_ScoreContainer {
   widgetClassPath=&HorizontalBox
   bIsVariable=true
   parent=Overlay_1
   widget=HorizontalBox_ScoreContainer
   properties {renderTransform {angle=0, scale=<1 1>, shear=<-10 0>, translation=<0 0>}, renderTransformPivot=<0.2 0.75>}
   slot {class=OverlaySlot, properties {verticalAlignment=VAlign_Bottom}}
  }
  CommonTextBlock_ScoreValue {
   widgetClassPath=&CommonTextBlock
   bIsVariable=true
   parent=HorizontalBox_ScoreContainer
   widget=CommonTextBlock_ScoreValue
   properties {
    bSimpleTextMode=true
    style=&TextStyle-ScorePopup_C
    text=+
    colorAndOpacity {colorUseRule=UseColor_Specified, specifiedColor=<rgba 0.1529 0.898 0.9882 1>}
    font {
     bForceMonospaced=false
     bMaterialIsStencil=false
     fontMaterial=null
     fontObject=&Burbank
     letterSpacing=0
     monospacedWidth=1
     size=38
     skewAmount=0
     typefaceFontName=Burbank-Black
     outlineSettings {
      bApplyOutlineToDropShadows=false
      bMiteredCorners=false
      bSeparateFillAlpha=false
      outlineMaterial=null
      outlineSize=0
      outlineColor=<rgba 0 0.0291 0.1562 1>
     }
    }
    renderTransformPivot=<0.2 0.75>
    shadowColorAndOpacity=<rgba 0.0052 0.0052 0.0056 0.55>
    shadowOffset=<3 3>
   }
   slot {class=HorizontalBoxSlot}
  }
  Number_Score {
   widgetClassPath=/Script/CommonUI.CommonNumericTextBlock
   bIsVariable=true
   parent=HorizontalBox_ScoreContainer
   widget=Number_Score
   properties {
    style=&TextStyle-ScorePopup_C
    text="1,000,000,000"
    colorAndOpacity {colorUseRule=UseColor_Specified, specifiedColor=<rgba 0.1529 0.898 0.9882 1>}
    font {
     bForceMonospaced=false
     bMaterialIsStencil=false
     fontMaterial=null
     fontObject=&Burbank
     letterSpacing=0
     monospacedWidth=1
     size=38
     skewAmount=0
     typefaceFontName=Burbank-Black
     outlineSettings {
      bApplyOutlineToDropShadows=false
      bMiteredCorners=false
      bSeparateFillAlpha=false
      outlineMaterial=null
      outlineSize=0
      outlineColor=<rgba 0 0.0291 0.1562 1>
     }
    }
    shadowColorAndOpacity=<rgba 0.0052 0.0052 0.0056 0.55>
    shadowOffset=<3 3>
   }
   slot {class=HorizontalBoxSlot, properties {horizontalAlignment=HAlign_Left, verticalAlignment=VAlign_Bottom}}
  }
  Image_Glow {
   widgetClassPath=&Image
   parent=Overlay_1
   widget=Image_Glow
   properties {
    renderOpacity=0
    brush {
     bIsDynamicallyLoaded=false
     drawAs=Image
     imageType=NoImage
     mirroring=NoMirror
     resourceName=None
     resourceObject=/Game/Athena/HUD/Materials/M_CountdownBG_SolidGlow
     tiling=NoTile
     imageSize=<429.8096 181.4673>
     margin=<ltrb 0 0 0 0>
     outlineSettings {
      bUseBrushTransparency=false
      roundingType=HalfHeightRadius
      width=0
      color {colorUseRule=UseColor_Specified, specifiedColor=<rgba 0 0 0 0>}
      cornerRadii=<0 0 0 0>
     }
     tintColor {colorUseRule=UseColor_Specified, specifiedColor=<rgba 0 0.2702 0.5729 1>}
     uVRegion {bIsValid=false, max=<0 0>, min=<0 0>}
    }
    colorAndOpacity=<rgba 1 1 1 0.1429>
    renderTransform {angle=0, scale=<1 1>, shear=<0 0>, translation=<8 0>}
   }
   slot {class=OverlaySlot, properties {horizontalAlignment=HAlign_Center, verticalAlignment=VAlign_Center, padding=<ltrb -250 -250 -250 -250>}}
  }
  CommonTextBlock_ScoreType {
   widgetClassPath=&CommonTextBlock
   bIsVariable=true
   parent=VerticalBox_1
   widget=CommonTextBlock_ScoreType
   properties {
    text=Message
    textTransformPolicy=ToUpper
    colorAndOpacity {colorUseRule=UseColor_Specified, specifiedColor=<rgba 0 0.491 1 1>}
    font {
     bForceMonospaced=false
     bMaterialIsStencil=false
     fontMaterial=null
     fontObject=&NotoSans
     letterSpacing=0
     monospacedWidth=1
     size=24
     skewAmount=0
     typefaceFontName=NotoSans-Bold
     outlineSettings {
      bApplyOutlineToDropShadows=false
      bMiteredCorners=false
      bSeparateFillAlpha=false
      outlineMaterial=null
      outlineSize=1
      outlineColor=<rgba 0 0.15 1 1>
     }
    }
    renderTransform {angle=0, scale=<1 1>, shear=<-10 0>, translation=<0 0>}
    shadowColorAndOpacity=<rgba 0.0052 0.0052 0.0056 0.55>
    shadowOffset=<4 3>
   }
   slot {class=VerticalBoxSlot, properties {horizontalAlignment=HAlign_Left, padding=<ltrb 0 -20 -500 0>}}
  }
  VerticalBox_Streak {
   widgetClassPath=&VerticalBox
   bIsVariable=true
   parent=VerticalBox_669
   widget=VerticalBox_Streak
   slot {class=VerticalBoxSlot, properties {horizontalAlignment=HAlign_Left, verticalAlignment=VAlign_Top, padding=<ltrb 0 -8 0 0>}}
  }
  SizeBox_2 {
   widgetClassPath=&SizeBox
   parent=VerticalBox_Streak
   widget=SizeBox_2
   properties {heightOverride=6.3701}
   slot {class=VerticalBoxSlot, properties {horizontalAlignment=HAlign_Left, verticalAlignment=VAlign_Top}}
  }
  Overlay_2 {
   widgetClassPath=&Overlay
   parent=SizeBox_2
   widget=Overlay_2
   slot {class=SizeBoxSlot, properties {verticalAlignment=VAlign_Bottom}}
  }
  HorizontalBox_1 {
   widgetClassPath=&HorizontalBox
   parent=Overlay_2
   widget=HorizontalBox_1
   slot {class=OverlaySlot, properties {verticalAlignment=VAlign_Bottom}}
  }
  VerticalBox_2 {
   widgetClassPath=&VerticalBox
   parent=HorizontalBox_1
   widget=VerticalBox_2
   properties {renderTransform {angle=0, scale=<1 1>, shear=<-12 0>, translation=<0 0>}}
   slot {class=HorizontalBoxSlot, properties {verticalAlignment=VAlign_Center}}
  }
  SizeBox_3 {
   widgetClassPath=&SizeBox
   parent=VerticalBox_2
   widget=SizeBox_3
   properties {bOverride_HeightOverride=true, bOverride_WidthOverride=true, heightOverride=12, widthOverride=200}
   slot {class=VerticalBoxSlot}
  }
  Image_MultiplierProgress {
   widgetClassPath=&Image
   bIsVariable=true
   parent=SizeBox_3
   widget=Image_MultiplierProgress
   properties {brush {
     bIsDynamicallyLoaded=false
     drawAs=Box
     imageType=NoImage
     mirroring=NoMirror
     resourceName=None
     resourceObject=/Game/Athena/HUD/Materials/Score/M_ScoreLTM_ScoreProgress_Square
     tiling=NoTile
     imageSize=<32 12>
     margin=<ltrb 0.1 0.1 0.1 0.1>
     outlineSettings {
      bUseBrushTransparency=false
      roundingType=HalfHeightRadius
      width=0
      color {colorUseRule=UseColor_Specified, specifiedColor=<rgba 0 0 0 0>}
      cornerRadii=<0 0 0 0>
     }
     tintColor {colorUseRule=UseColor_Specified, specifiedColor=<rgba 1 1 1 1>}
     uVRegion {bIsValid=false, max=<0 0>, min=<0 0>}
    }}
   slot {class=SizeBoxSlot}
  }
  SizeBox {
   widgetClassPath=&SizeBox
   parent=VerticalBox_2
   widget=SizeBox
   properties {bOverride_HeightOverride=true, bOverride_WidthOverride=true, heightOverride=5, widthOverride=200}
   slot {class=VerticalBoxSlot}
  }
  Image_TimeRemaining {
   widgetClassPath=&Image
   bIsVariable=true
   parent=SizeBox
   widget=Image_TimeRemaining
   properties {brush {
     bIsDynamicallyLoaded=false
     drawAs=Box
     imageType=NoImage
     mirroring=NoMirror
     resourceName=None
     resourceObject=/Game/Athena/HUD/Materials/Score/M_ScoreLTM_ScoreProgress_Square_WhiteFill
     tiling=NoTile
     imageSize=<32 12>
     margin=<ltrb 0.1 0.1 0.1 0.1>
     outlineSettings {
      bUseBrushTransparency=false
      roundingType=HalfHeightRadius
      width=0
      color {colorUseRule=UseColor_Specified, specifiedColor=<rgba 0 0 0 0>}
      cornerRadii=<0 0 0 0>
     }
     tintColor {colorUseRule=UseColor_Specified, specifiedColor=<rgba 1 1 1 1>}
     uVRegion {bIsValid=false, max=<0 0>, min=<0 0>}
    }}
   slot {class=SizeBoxSlot}
  }
  Overlay_3 {
   widgetClassPath=&Overlay
   parent=HorizontalBox_1
   widget=Overlay_3
   slot {class=HorizontalBoxSlot, properties {padding=<ltrb 0 -10 0 -10>}}
  }
  HUD_Mash_MultiplierEffects {
   widgetClassPath=/Game/Athena/HUD/LTM/Mash/HUD_Mash_MultiplierEffects
   bIsVariable=true
   parent=Overlay_3
   widget=HUD_Mash_MultiplierEffects
   slot {class=OverlaySlot, properties {horizontalAlignment=HAlign_Center, verticalAlignment=VAlign_Center}}
  }
  CommonTextBlock_MultiplierValue {
   widgetClassPath=&CommonTextBlock
   bIsVariable=true
   parent=Overlay_3
   widget=CommonTextBlock_MultiplierValue
   properties {
    text=x4
    colorAndOpacity {colorUseRule=UseColor_Specified, specifiedColor=<rgba 0.6867 0.8714 1 1>}
    font {
     bForceMonospaced=false
     bMaterialIsStencil=false
     fontMaterial=null
     fontObject=&BurbankSmall
     letterSpacing=0
     monospacedWidth=1
     size=30
     skewAmount=0
     typefaceFontName=Black
     outlineSettings {
      bApplyOutlineToDropShadows=false
      bMiteredCorners=false
      bSeparateFillAlpha=false
      outlineMaterial=/Game/Athena/HUD/Materials/M_UI_OutlineShine_White_Manual
      outlineSize=1
      outlineColor=<rgba 0.0052 0.0052 0.0056 0.55>
     }
    }
    renderTransform {angle=0, scale=<1 1>, shear=<-12 0>, translation=<0 0>}
    renderTransformPivot=<0.2 0.75>
    shadowColorAndOpacity=<rgba 0.0056 0.0052 0.0014 0.55>
    shadowOffset=<3 3>
   }
   slot {class=OverlaySlot}
  }
  Image_StreakGlow {
   widgetClassPath=&Image
   parent=Overlay_3
   widget=Image_StreakGlow
   properties {
    renderOpacity=0
    brush {
     bIsDynamicallyLoaded=false
     drawAs=Image
     imageType=NoImage
     mirroring=NoMirror
     resourceName=None
     resourceObject=/Game/Athena/HUD/Materials/Victory/M_WinGlow_White_Inst
     tiling=NoTile
     imageSize=<605.409 235.7135>
     margin=<ltrb 0 0 0 0>
     outlineSettings {
      bUseBrushTransparency=false
      roundingType=HalfHeightRadius
      width=0
      color {colorUseRule=UseColor_Specified, specifiedColor=<rgba 0 0 0 0>}
      cornerRadii=<0 0 0 0>
     }
     tintColor {colorUseRule=UseColor_Specified, specifiedColor=<rgba 1 1 1 1>}
     uVRegion {bIsValid=false, max=<0 0>, min=<0 0>}
    }
    colorAndOpacity=<rgba 0.0264 0.0145 0.2917 1>
    renderTransform {angle=0, scale=<0.5 0.5>, shear=<0 0>, translation=<4 0>}
   }
   slot {class=OverlaySlot, properties {horizontalAlignment=HAlign_Center, verticalAlignment=VAlign_Center, padding=<ltrb -300 -180 -300 -180>}}
  }
  CommonTextBlock_MultiplierType {
   widgetClassPath=&CommonTextBlock
   bIsVariable=true
   parent=VerticalBox_Streak
   widget=CommonTextBlock_MultiplierType
   properties {
    text=Score Streak
    textTransformPolicy=ToUpper
    colorAndOpacity {colorUseRule=UseColor_Specified, specifiedColor=<rgba 0.6867 0.8714 1 1>}
    font {
     bForceMonospaced=false
     bMaterialIsStencil=false
     fontMaterial=null
     fontObject=&BurbankSmall
     letterSpacing=0
     monospacedWidth=1
     size=24
     skewAmount=0
     typefaceFontName=Black
     outlineSettings {
      bApplyOutlineToDropShadows=false
      bMiteredCorners=false
      bSeparateFillAlpha=false
      outlineMaterial=null
      outlineSize=1
      outlineColor=<rgba 0 0 0 1>
     }
    }
    renderTransform {angle=0, scale=<1 1>, shear=<-12 0>, translation=<0 0>}
    shadowColorAndOpacity=<rgba 0.0056 0.0052 0.0014 0.55>
    shadowOffset=<4 3>
   }
   slot {class=VerticalBoxSlot, properties {horizontalAlignment=HAlign_Left, padding=<ltrb 0 -11 -500 0>}}
  }
 ]
}
